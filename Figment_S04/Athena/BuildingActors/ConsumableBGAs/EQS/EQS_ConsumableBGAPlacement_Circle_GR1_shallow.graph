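# EQS query that picks placement points for consumable BGAs (building game actors) in a circular area around the querier.
# A donut generator lays evenly spaced concentric rings of points and projects them onto geometry with a vertically limited channel trace.
# A distance test filters out points too close to the querier in 3D, while a Fort random test supplies the scoring so selection varies.
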EnvQuery /Figment_S04/Athena/BuildingActors/ConsumableBGAs/EQS/EQS_ConsumableBGAPlacement_Circle_GR1_shallow
options [{
  generator {
   name=EnvQueryGenerator_Donut_1
   class=EnvQueryGenerator_Donut
   properties {
    bUseSpiralPattern=false
    center=&EnvQueryContext_Querier
    innerRadius=200
    numberOfRings=4
    outerRadius=1800
    pointsPerRing=10
    projectionData {extentX=0, navigationFilter=null, postProjectionVerticalOffset=0, projectDown=512, projectUp=256, traceMode=GeometryByChannel}
   }
  }
  tests [
   FortQueryTest_Random_0 {
    class=FortQueryTest_Random
    properties {
     bDefineReferenceValue=true
     bUseRandomSeedForOthers=true
     clampMaxType=SpecifiedValue
     clampMinType=SpecifiedValue
     floatValueMax=1
     floatValueMin=1
     multipleContextScoreOp=MaxScore
     scoreClampMax=1
     scoringFactor=-1
    }
   }
   EnvQueryTest_Distance_2 {
    class=EnvQueryTest_Distance
    properties {
     distanceTo=&EnvQueryContext_Querier
     filterType=Minimum
     floatValueMax=1000
     floatValueMin=400
     scoringEquation=Square
     scoringFactor=1
     testComment=""
     testMode=Distance3D
     testPurpose=Filter
    }
   }
  ]
 }]
