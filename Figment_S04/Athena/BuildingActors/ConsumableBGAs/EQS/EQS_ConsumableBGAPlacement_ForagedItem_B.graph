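# EQS query that picks placement points for foraged-item consumable building actors, judging by its name, in a cone around the querier.
# A cone generator, with query-parameter bindings for range, spacing, angle step and cone width, projects points onto geometry using visibility line traces.
# A 2D distance test filters out points closer than a parameterized minimum, and a Fortnite random test scores points on a square-root curve to vary placement.
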EnvQuery /Figment_S04/Athena/BuildingActors/ConsumableBGAs/EQS/EQS_ConsumableBGAPlacement_ForagedItem_B
options [{
  generator {
   name=EnvQueryGenerator_Cone_0
   class=EnvQueryGenerator_Cone
   properties {
    bAutoSortTests=true
    bCanRunAsync=false
    bIncludeContextLocation=false
    centerActor=&EnvQueryContext_Querier
    envQueryResultNormalizationOption=Default
    navDataOverrideContext=null
    optionName=""
    alignedPointsDistance {
     dataBinding=&self:EnvQueryGenerator_Cone_0.AIDataProvider_QueryParams_7
     dataField=FloatValue
     defaultValue=400
    }
    angleStep {
     dataBinding=&self:EnvQueryGenerator_Cone_0.AIDataProvider_QueryParams_9
     dataField=FloatValue
     defaultValue=20
    }
    coneDegrees {
     dataBinding=&self:EnvQueryGenerator_Cone_0.AIDataProvider_QueryParams_4
     dataField=FloatValue
     defaultValue=90
    }
    projectionData {
     bCanDisableTrace=true
     bCanProjectDown=true
     bCanTraceOnGeometry=true
     bCanTraceOnNavMesh=true
     bOnlyBlockingHits=true
     bTraceComplex=false
     extentX=0
     extentY=10
     extentZ=10
     navigationFilter=null
     postProjectionVerticalOffset=0
     projectDown=512
     projectUp=128
     serializedChannel=ECC_Visibility
     traceChannel=TraceTypeQuery1
     traceMode=GeometryByChannel
     traceProfileName=None
     traceShape=Line
     versionNum=1
    }
    range {
     dataBinding=&self:EnvQueryGenerator_Cone_0.AIDataProvider_QueryParams_8
     dataField=FloatValue
     defaultValue=1500
    }
   }
  }
  tests [
   FortQueryTest_Random_1 {class=FortQueryTest_Random, properties {bUseRandomSeedForOthers=true, floatValueMax=1, scoringEquation=SquareRoot, scoringFactor=1}}
   EnvQueryTest_Distance_0 {
    class=EnvQueryTest_Distance
    properties {
     distanceTo=&EnvQueryContext_Querier
     filterType=Minimum
     scoringEquation=Linear
     scoringFactor=1
     testComment=""
     testMode=Distance2D
     testPurpose=Filter
     floatValueMin {
      dataBinding=&self:EnvQueryTest_Distance_0.AIDataProvider_QueryParams_2
      dataField=FloatValue
      defaultValue=0
     }
    }
   }
  ]
 }]
