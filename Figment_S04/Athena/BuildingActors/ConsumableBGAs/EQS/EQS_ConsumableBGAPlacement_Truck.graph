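# EQS query that picks a placement point for a consumable building actor, apparently the Truck variant going by its name.
# A simple grid generator lays out candidate points and projects them downward onto geometry by channel trace.
# A single Fort random test then filters and scores the points, with its random-seed sharing option enabled, so selection is randomized rather than based on spatial criteria.
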EnvQuery /Figment_S04/Athena/BuildingActors/ConsumableBGAs/EQS/EQS_ConsumableBGAPlacement_Truck
options [{
  generator {
   name=EnvQueryGenerator_SimpleGrid_0
   class=EnvQueryGenerator_SimpleGrid
   properties {
    gridSize=160
    spaceBetween=70
    projectionData {extentX=0, navigationFilter=null, postProjectionVerticalOffset=0, projectDown=128, projectUp=0, traceMode=GeometryByChannel}
   }
  }
  tests [FortQueryTest_Random_0 {class=FortQueryTest_Random, properties {bUseRandomSeedForOthers=true, floatValueMax=1, scoringFactor=1, testPurpose=FilterAndScore}}]
 }]
